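# MetaSound Source named as boss music for the Adventure Girl NPC in Clyde, currently an empty scaffold.
# It declares a mono, one-shot interface with an On Play trigger input, an audio output and an On Finished trigger output.
# The graph holds no nodes or connections, so it generates no sound and never signals completion.
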MetaSoundSource /Clyde/Audio/BossMusic/MSS_NPCMusic_Clyde_Boss_AdventureGirl
inputs [UE.Source.OnPlay {type=Trigger}]
interfaces [UE.OutputFormat.Mono:1.0, UE.Source.OneShot:1.0, UE.Source:1.1]
outputs [
 "UE.OutputFormat.Mono.Audio:0" {type=Audio}
 UE.Source.OneShot.OnFinished {type=Trigger}
]
pages [Default {input_defaults {UE.Source.OnPlay=null}, links [], nodes []}]
properties {assetImportData=&self:AssetImportData}
